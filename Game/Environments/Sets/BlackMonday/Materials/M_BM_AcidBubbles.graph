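# Translucent, two-sided lit material for acid bubble mesh particles in the BlackMonday environment set, previewed on a half sphere.
# A double-sided Fresnel with an exponent parameter blends from one to the FresnelInner value, separating rim from interior.
# Particle color RGB is multiplied by itself and fed through an FX emissive adjustment function.
# A scalar Refraction parameter accompanies pixel normal offset refraction, and responsive AA is enabled.
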Material /Game/Environments/Sets/BlackMonday/Materials/M_BM_AcidBubbles
graph {
 properties {
  bEnableResponsiveAA=true
  bUsedWithMeshParticles=true
  blendMode=BLEND_Translucent
  pixelDepthOffsetMode=PDOM_Legacy
  previewMesh=/Game/Effects/Fort_Effects/Mesh/SM_HalfSphere_100
  refractionMethod=RM_PixelNormalOffset
  translucencyLightingMode=TLM_VolumetricPerVertexDirectional
  twoSided=true
  usedShadingModels=Default Lit
 }
 nodes [
  LinearInterpolate(
   B=ScalarParameter() {defaultValue=0.1, parameterName=FresnelInner}
   Alpha=MaterialFunctionCall(Exponent=ScalarParameter() {defaultValue=0.1, parameterName=FresnelExponent}) {materialFunction=&MF_Fresnel_FX}.Fresnel Double Sided
  ) {constA=1}
  n1 = ParticleColor()
  MaterialFunctionCall(Emissive=Multiply(A=@n1.RGB, B=@n1.RGB)) {materialFunction=&MF_FX_EmissiveAdjustments}
  ScalarParameter() {defaultValue=1.1, parameterName=Refraction}
 ]
}
