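# Unlit translucent particle-sprite material that, by its name, draws ripples for acid bubbles in the BlackMonday set.
# Particle color RGB passes through the shared FX emissive adjustment function.
# A noise sample, shifted by dynamic UVBias and scaled by RippleStrength, distorts the UVs of a centered circular SphereMask.
# A second SphereMask reshapes that mask, with the dynamic Center output feeding its center, radius and hardness, before multiplying by particle alpha.
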Material /Game/Environments/Sets/BlackMonday/Materials/M_BM_AcidBubbles_Ripple
graph {
 properties {
  bUsedWithParticleSprites=true
  blendMode=BLEND_Translucent
  pixelDepthOffsetMode=PDOM_Legacy
  shadingModel=MSM_Unlit
  usedShadingModels=Unlit
 }
 nodes [
  n1 = ParticleColor()
  MaterialFunctionCall(Emissive=@n1.RGB) {materialFunction=&MF_FX_EmissiveAdjustments}
  n2 = TextureSample(UVs=Add(A=DynamicParameter() {defaultValue=<rgba 1 1 0 1>, paramNames [Center,Radius,Hardness,UVBias]}.UVBias, B=TextureCoordinate())) {samplerType=SAMPLERTYPE_LinearColor, texture=&T_NoisePack_LinearColor}
  n3 = AppendVector(A=@n2.R, B=@n2.R)
  n4 = Add(A=Multiply(A=ScalarParameter() {defaultValue=0.1, parameterName=RippleStrength}, B=Subtract(A=@n3) {constB=0.5}), B=TextureCoordinate())
  n5 = Constant() {r=0.5}
  n6 = Constant() {r=0.45}
  n7 = Constant()
  n8 = DynamicParameter() {defaultValue=<rgba 1 1 0 1>, paramNames [Center,Radius,Hardness,UVBias]}
  Multiply(A=@n1.A, B=SphereMask(A=SphereMask(A=@n4, B=@n5, Radius=@n6, Hardness=@n7), B=@n8.Center, Radius=@n8.Center, Hardness=@n8.Center))
 ]
}
