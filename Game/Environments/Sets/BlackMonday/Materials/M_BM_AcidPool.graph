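# Acid pool surface material for the BlackMonday environment set, a lit shader with glowing, green-by-default liquid.
# Parallax bump-offset bubbles, borrowed from a SlurpMonster wrap texture, tint and break up the base color with static noise.
# Emissive adds flow-map-warped panning noise, an inner glow and parallax caustics distorted by noise and two blended panning ripple normals.
# Sphere masks on the second UV set's V coordinate shape the glow and fade color and emissive.
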Material /Game/Environments/Sets/BlackMonday/Materials/M_BM_AcidPool
graph {
 properties {pixelDepthOffsetMode=PDOM_Legacy, usedShadingModels=Default Lit}
 nodes [
  n1 = Multiply(A=TextureCoordinate(), B=ScalarParameter() {defaultValue=6, parameterName=BubblesTile})
  n2 = TextureSample() {samplerType=SAMPLERTYPE_LinearColor, texture=/Game/Global/Textures/Default/Noise/T_Noise01_Inverse}
  n3 = LinearInterpolate(A=ScalarParameter() {defaultValue=-10, parameterName=BubblesBumpOffsetHeight}, Alpha=@n2.R) {constB=0}
  n4 = VectorParameter() {parameterName=BaseColor, defaultValue=<rgba 0.2268 0.3594 0.0181 0>}
  n5 = Multiply(
   A=TextureSample(UVs=BumpOffset(Coordinate=@n1, Height=@n3)) {samplerType=SAMPLERTYPE_Masks, texture=/Game/Weapons/WeaponSkins/Wraps/Textures/SlurpMonster/T_Wrap_SlurpMonster_Bubbles_MB}.R
   B=@n4.RGB
  )
  n6 = SphereMask(
   A=MaterialFunctionCall(Float2=TextureCoordinate() {coordinateIndex=1}) {materialFunction=&BreakOutFloat2Components}.G
   B=Constant() {r=1}
   Radius=Constant() {r=1}
   Hardness=Constant() {r=0.9}
  )
  Multiply(A=Multiply(A=Add(A=@n5, B=@n4.RGB), B=@n2.R), B=@n6)
  n7 = SphereMask(
   A=MaterialFunctionCall(Float2=TextureCoordinate() {coordinateIndex=1}) {materialFunction=&BreakOutFloat2Components}.G
   B=Constant() {r=1}
   Radius=ScalarParameter() {defaultValue=1, parameterName=InnerGlowRadius}
   Hardness=ScalarParameter() {defaultValue=0.03, parameterName=InnerGlowHardness}
  )
  n8 = VectorParameter() {parameterName=NoisePanTile, defaultValue=<rgba 2 0.5 0 0.011>}
  n9 = TextureCoordinate() {coordinateIndex=1}
  n10 = VectorParameter() {parameterName=FlowMapPanTile, defaultValue=<rgba 1 0.2 0 0.01>}
  n11 = AppendVector(A=@n10.R, B=@n10.R)
  n12 = TextureSample(UVs=Panner(Coordinate=Multiply(A=@n9, B=@n11), Speed=AppendVector(A=@n10.B, B=@n10.B))) {samplerType=SAMPLERTYPE_LinearColor, texture=/Game/Global/Textures/Default/Flowmaps/T_SwirlFlow_02}
  n13 = Add(
   A=Multiply(A=TextureCoordinate() {coordinateIndex=1}, B=AppendVector(A=@n8.R, B=@n8.R))
   B=Multiply(A=AppendVector(A=@n12.R, B=@n12.R), B=ScalarParameter() {defaultValue=0.6, parameterName=FlowScale})
  )
  n14 = AppendVector(A=@n8.B, B=@n8.B)
  n15 = Power(
   Base=TextureSample(UVs=Panner(Coordinate=@n13, Speed=@n14)) {samplerType=SAMPLERTYPE_LinearColor, texture=&T_Noise01_Inverse}.R
   Exp=ScalarParameter() {defaultValue=3, parameterName=NoiseExponent}
  )
  n16 = ScalarParameter() {defaultValue=0.2, parameterName=NoiseScale}
  n17 = Power(Base=@n7, Exp=ScalarParameter() {defaultValue=2.5, parameterName=InnerGlowExponent})
  n18 = LinearInterpolate(
   A=Multiply(A=@n7, B=Multiply(A=@n15, B=@n16))
   B=Multiply(A=@n17, B=ScalarParameter() {defaultValue=0.6, parameterName=InnerBrightness})
   Alpha=@n17
  )
  n19 = TextureCoordinate()
  n20 = Multiply(A=TextureCoordinate(), B=ScalarParameter() {defaultValue=1.5, parameterName=CausticNoiseScaleUV})
  n21 = VectorParameter() {parameterName=CausticNoisePannerSpeeds, defaultValue=<rgba -0.056 0.0109 0.0224 -0.0096>}
  n22 = AppendVector(A=@n21.R, B=@n21.R)
  n23 = AppendVector(A=@n21.B, B=@n21.B)
  n24 = AppendVector(
   A=TextureSample(UVs=Panner(Coordinate=@n20, Speed=@n22)) {samplerType=SAMPLERTYPE_LinearColor, texture=&T_NoisePack_LinearColor}.R
   B=TextureSample(UVs=Panner(Coordinate=@n20, Speed=@n23)) {samplerType=SAMPLERTYPE_LinearColor, texture=&T_NoisePack_LinearColor}.G
  )
  n25 = ScalarParameter() {defaultValue=1, parameterName=NormalA_TileU, sliderMin=1}
  n26 = ScalarParameter() {defaultValue=2, parameterName=NormalA_TileV}
  n27 = Panner(
   Coordinate=Multiply(A=AppendVector(A=@n25, B=@n26), B=TextureCoordinate() {coordinateIndex=1})
   Speed=AppendVector(A=ScalarParameter() {parameterName=NormalA_PanSpeedU}, B=ScalarParameter() {defaultValue=0.04, parameterName=NormalA_PanSpeedV})
  ) {bFractionalPart=true}
  n28 = ScalarParameter() {defaultValue=1, parameterName=NormalA_TileU, sliderMin=1}
  n29 = ScalarParameter() {defaultValue=2, parameterName=NormalA_TileV}
  n30 = Panner(
   Coordinate=Multiply(A=AppendVector(A=@n28, B=@n29), B=TextureCoordinate() {coordinateIndex=1})
   Speed=AppendVector(A=ScalarParameter() {parameterName=NormalB_PanSpeedU}, B=ScalarParameter() {defaultValue=0.0266, parameterName=NormalB_PanSpeedV})
  ) {bFractionalPart=true}
  n31 = Multiply(
   A=MaterialFunctionCall(
    BaseNormal=TextureSample(UVs=@n27) {samplerType=SAMPLERTYPE_Normal, texture=/Game/Effects/Fort_Effects/Gameplay/Water/Textures/T_WaterRipples_01}.RGB
    AdditionalNormal=TextureSample(UVs=@n30) {samplerType=SAMPLERTYPE_Normal, texture=&T_WaterRipples_01}.RGB
   ) {materialFunction=&BlendAngleCorrectedNormals}.Result
   B=VectorParameter() {parameterName=NormalIntensity, defaultValue=<rgba 0.08 0.08 1 0>}.RGB
  )
  n32 = Multiply(
   A=ScalarParameter() {defaultValue=0.05, parameterName=CausticsDistortionScale}
   B=Add(A=Subtract(A=@n24) {constB=0.5}, B=ComponentMask(None=@n31) {g=true, r=true})
  )
  n33 = Multiply(
   A=ScalarParameter() {defaultValue=1.5, parameterName=CausticsTile}
   B=BumpOffset(Coordinate=Add(A=@n19, B=@n32), Height=ScalarParameter() {defaultValue=-5, parameterName=CausticsBumpOffsetHeight})
  )
  n34 = Multiply(
   A=Power(
    Base=TextureSample(UVs=@n33) {samplerType=SAMPLERTYPE_Masks, texture=&T_Wrap_SlurpMonster_Bubbles_MB}.G
    Exp=ScalarParameter() {defaultValue=3, parameterName=CausticsExponent}
   )
   B=ScalarParameter() {defaultValue=0.24, parameterName=CausticsScale}
  )
  Multiply(A=@n6, B=Multiply(A=VectorParameter() {parameterName=EmissiveColor, defaultValue=<rgba 0.28 1.1 0.0859 0>}.RGB, B=Add(A=@n18, B=@n34)))
  ScalarParameter() {defaultValue=0.6, parameterName=Metallic}
  ScalarParameter() {defaultValue=0.85, parameterName=Specular}
  ScalarParameter() {defaultValue=0.2, parameterName=Roughness}
 ]
}
