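# Translucent unlit particle material for acid smoke in the BlackMonday set, usable on sprites and Niagara sprites or meshes.
# Particle color feeds an emissive adjustment function, while its alpha scales a tiled inky smoke texture masked by soft smoke.
# Dynamic parameters set tiling, contrast exponent and random offset, with URandom driving both axes and VRandom unused.
# A quality switch adds depth fade and near-camera fade on Medium and above.
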Material /Game/Environments/Sets/BlackMonday/Materials/M_BM_AcidSmoke
graph {
 properties {
  bEnableResponsiveAA=true
  bUsedWithNiagaraMeshParticles=true
  bUsedWithNiagaraSprites=true
  bUsedWithParticleSprites=true
  blendMode=BLEND_Translucent
  pixelDepthOffsetMode=PDOM_Legacy
  shadingModel=MSM_Unlit
  usedShadingModels=Unlit
 }
 nodes [
  MaterialFunctionCall(Emissive=ParticleColor().RGB) {materialFunction=&MF_FX_EmissiveAdjustments}
  n1 = TextureCoordinate()
  n2 = DynamicParameter() {defaultValue=<rgba 0 0 1 1>, paramNames [URandom,VRandom,Tile,Exponent]}
  n3 = TextureSample(UVs=Add(A=Multiply(A=@n1, B=@n2.Tile), B=AppendVector(A=@n2.URandom, B=@n2.URandom))) {texture=&T_Inky_Smoke_Tile}
  n4 = DynamicParameter() {defaultValue=<rgba 0 0 1 1>, paramNames [URandom,VRandom,Tile,Exponent]}
  n5 = Multiply(
   A=ParticleColor().A
   B=Multiply(A=Power(Base=@n3.RGB, Exp=@n4.Exponent), B=TextureSample() {texture=/Game/Effects/Fort_Effects/Textures/T_Smoke_Soft}.RGB)
  )
  n6 = Multiply(
   A=DepthFade(Opacity=@n5, FadeDistance=ScalarParameter() {defaultValue=50, parameterName=DepthFadeDistance})
   B=MaterialFunctionCall(Distance=ScalarParameter() {parameterName=NearCameraDistance}) {materialFunction=&Near_Camera_Fade}.Result
  )
  QualitySwitch(Default=@n5, Low=@n5, High=@n6, Medium=@n6, Epic=@n6)
 ]
}
