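# Additive unlit fog card for the BlackMonday environment set, drawn as a soft, camera-facing glow.
# Panning cloud noise, offset by per-primitive Random, perturbs a radial mask that is shaped by a power and tinted by per-primitive Color.
# Color alpha, scene-depth fade and a pivot-to-camera distance ramp drive opacity.
# World position offset billboards the card and nudges it cameraward, collapsing it to a point when the camera is near.
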Material /Game/Environments/Sets/BlackMonday/Materials/M_BM_FogCard
graph {
 properties {bEnableResponsiveAA=true, blendMode=BLEND_Additive, pixelDepthOffsetMode=PDOM_Legacy, shadingModel=MSM_Unlit, usedShadingModels=Unlit}
 nodes [
  n1 = ScalarParameter() {defaultValue=200, parameterName=NearCameraScale}
  n2 = MaterialFunctionCall!(
   min=@n1
   max=ScalarParameter() {defaultValue=2000, parameterName=FarCameraScale}
   value=Distance(
    A=MaterialFunctionCall() {materialFunction=&ObjectPivotPoint}.Object Pivot Location
    B=CameraPositionWS()
   )
  ) {materialFunction=/Game/Effects/Fort_Effects/Materials/_Functions/InverseLerp_MF}
  n3 = WorldPosition()
  n4 = Constant3Vector() {constant=<rgba 0 1 0 1>}
  n5 = Constant3Vector() {constant=<rgba 0 0 -1 0>}
  n6 = Normalize(VectorInput=Subtract(A=CameraPositionWS(), B=ObjectPositionWS()))
  n7 = ScalarParameter() {defaultValue=200, parameterName=NearCameraScale}
  n8 = Subtract(A=ObjectPositionWS(), B=WorldPosition().XYZ)
  If(
   A=@n2.Clamped
   B=Constant() {r=0.01}
   "A > B"=Add(
    A=MaterialFunctionCall!(WorldPosition=@n3.XYZ, PivotLocation, WS Normals, Custom Object Basis 1=@n4, Custom Object Basis 2=@n5, Custom Object Basis 3) {materialFunction=/Engine/Functions/Engine_MaterialFunctions02/WorldPositionOffset/AlignMeshToTheCamera}.World Position Offset
    B=Multiply(A=@n6, B=@n7)
   )
   "A == B"=@n8
   "A < B"=@n8
  )
  n9 = VectorParameter() {bUseCustomPrimitiveData=true, parameterName=Color, primitiveDataIndex=28, defaultValue=<rgba 0.3 0.1538 0.0802 0>}
  n10 = ScalarParameter() {bUseCustomPrimitiveData=true, parameterName=Random, primitiveDataIndex=27}
  n11 = TextureCoordinate()
  n12 = Panner(
   Coordinate=Multiply(A=Add(A=@n10, B=@n11), B=ScalarParameter() {defaultValue=0.22, parameterName=NoiseScaleUV})
   Speed=AppendVector(
    A=ScalarParameter() {defaultValue=-0.0034, parameterName=NoisePanSpeedU}
    B=ScalarParameter() {defaultValue=0.0113, parameterName=NoisePanSpeedV}
   )
  ) {bFractionalPart=true}
  n13 = Multiply(
   A=Subtract(A=TextureSample(UVs=@n12) {texture=&AtmosphericCloudNoise01}.R) {constB=0.5}
   B=ScalarParameter() {defaultValue=0.4, parameterName=NoiseScale}
  )
  n14 = Constant() {r=0.5}
  n15 = SphereMask(A=TextureCoordinate(), B=@n14, Radius=@n14, Hardness=Constant())
  Multiply(A=@n9.RGB, B=Power(Base=Add(A=@n13, B=@n15), Exp=ScalarParameter() {defaultValue=2.5, parameterName=ExponentFalloff}))
  n16 = Divide(A=@n1)
  n17 = Subtract(A=Subtract(A=SceneDepth().R, B=@n16), B=Subtract(A=PixelDepth().R, B=@n16))
  n18 = ScalarParameter() {defaultValue=200, parameterName=DepthFadeDistance}
  Multiply(A=Saturate(None=Divide(A=@n17, B=@n18)), B=VertexInterpolator(VS=Multiply(A=@n9.A, B=@n2.Clamped)).PS)
 ]
}
