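# Lit surface material for string lights in the BlackMonday environment set, enabled for Nanite and spline meshes.
# The specular texture's alpha masks a warm, overbright Emissive color parameter, which multiplies the diffuse texture's RGB to form the tinted glow.
# A normal map supplies surface detail.
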Material /Game/Environments/Sets/BlackMonday/Materials/M_BM_StringLight_SM01
graph {
 properties {bUsedWithNanite=true, bUsedWithSplineMeshes=true, pixelDepthOffsetMode=PDOM_Legacy, usedShadingModels=Default Lit}
 nodes [
  Multiply(
   A=TextureSample() {texture=/Game/Environments/Sets/BlackMonday/Textures/T_BM_StringLights_01_D}.RGB
   B=Multiply(
    A=TextureSample() {samplerType=SAMPLERTYPE_Masks, texture=/Game/Environments/Sets/BlackMonday/Textures/T_BM_StringLights_01_S}.A
    B=VectorParameter() {parameterName=Emissive, defaultValue=<rgba 20 9.9456 3.3946 1>}.RGB
   )
  )
  TextureSample() {samplerType=SAMPLERTYPE_Normal, texture=/Game/Environments/Sets/BlackMonday/Textures/T_BM_StringLights_01_N}
 ]
}
