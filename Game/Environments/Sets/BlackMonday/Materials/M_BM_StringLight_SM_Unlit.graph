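# Material for the Black Monday set's string lights, blinking amber, magenta and blue bulbs.
# Vertex color channels select each color, driven by time-offset sines snapped through Floor/Ceil into square waves, then raised to a power.
# Camera distance drives a normal-direction offset, masked by red and blue only (red repeated, green absent), plus an emissive dimming term.
# Shading-path, quality and IsLit switches choose dimmed or full glow; diffuse, mask and normal textures are sampled.
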Material /Game/Environments/Sets/BlackMonday/Materials/M_BM_StringLight_SM_Unlit
graph {
 properties {bUsedWithNanite=true, bUsedWithSplineMeshes=true, pixelDepthOffsetMode=PDOM_Legacy, usedShadingModels=Default Lit}
 nodes [
  n1 = Constant()
  n2 = Constant() {r=0.4}
  n3 = Time()
  n4 = Sine(None=Multiply(A=Add(A=@n2, B=@n3), B=Constant() {r=1}))
  n5 = Floor(None=@n4)
  n6 = Ceil(None=@n4)
  n7 = VertexColor()
  n8 = LinearInterpolate(A=Constant(), B=Multiply(A=Constant3Vector() {constant=<rgba 1.5 1.067 0 0>}, B=If(A=@n1, B=@n4, "A > B"=@n5, "A < B"=@n6)), Alpha=@n7.B)
  n9 = Constant()
  n10 = Sine(None=Multiply(A=Time(), B=Constant() {r=1}))
  n11 = Floor(None=@n10)
  n12 = Ceil(None=@n10)
  n13 = LinearInterpolate(A=Constant(), B=Multiply(A=Constant3Vector() {constant=<rgba 1.5 0 1.372 0>}, B=If(A=@n9, B=@n10, "A > B"=@n11, "A < B"=@n12)), Alpha=@n7.R)
  n14 = Constant()
  n15 = Constant() {r=0.2}
  n16 = Time()
  n17 = Sine(None=Multiply(A=Add(A=@n15, B=@n16), B=Constant() {r=1}))
  n18 = Floor(None=@n17)
  n19 = Ceil(None=@n17)
  n20 = LinearInterpolate(
   A=Constant()
   B=Multiply(A=Constant3Vector() {constant=<rgba 0 0.0305 1.5 0>}, B=If(A=@n14, B=@n17, "A > B"=@n18, "A < B"=@n19))
   Alpha=@n7.G
  )
  Add(A=@n8, B=Add(A=@n13, B=@n20))
  n21 = TextureSample() {samplerType=SAMPLERTYPE_Masks, texture=/Game/Environments/Sets/BlackMonday/Textures/T_BM_StringLights_01_S}
  LinearInterpolate(Alpha=@n21.A)
  n22 = VertexNormalWS()
  n23 = ScalarParameter() {defaultValue=24, parameterName=Scale}
  n24 = Distance(A=WorldPosition().XYZ, B=CameraPositionWS())
  n25 = ScalarParameter() {defaultValue=2048, parameterName=DistanceBias}
  n26 = Clamp(None=Divide(A=Subtract(A=@n24, B=@n25), B=ScalarParameter() {defaultValue=7000, parameterName=FadeDistance}))
  LinearInterpolate(B=LinearInterpolate(B=Multiply(A=@n22, B=@n23), Alpha=@n26), Alpha=Max(A=Max(A=@n7.R, B=@n7.R), B=@n7.B))
  Multiply(A=TextureSample() {texture=/Game/Environments/Sets/BlackMonday/Textures/T_BM_StringLights_01_D}.RGB, B=Multiply(A=@n21.A))
  Normalize(VectorInput=Subtract(A=WorldPosition().XYZ, B=ObjectPositionWS()))
  n27 = Power(Base=LinearInterpolate(A=LinearInterpolate(A=@n13, B=@n20, Alpha=@n7.G), B=@n8, Alpha=@n7.B), Exp=Constant() {r=10})
  n28 = Multiply(A=@n27, B=LinearInterpolate(B=ScalarParameter() {defaultValue=0.25, parameterName=MinEmissive}, Alpha=@n26) {constA=1, constB=0.1})
  ShadingPathSwitch(Default=QualitySwitch(Default=@n27), Mobile=QualitySwitch(Default=@n27, Low=@n28, Medium=@n28))
  StaticSwitch(True=@n28, Value=StaticBoolParameter() {parameterName=IsLit})
  TextureSample() {samplerType=SAMPLERTYPE_Normal, texture=/Game/Environments/Sets/BlackMonday/Textures/T_BM_StringLights_01_N}
 ]
}
